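# Sound cue for the third 1000-series tentacle attack sound in Section 02 of the PistolCanary event.
# A single wave player feeds the output directly, with no randomization or modulation nodes.
# It uses the PistolCanary tentacle-tip attenuation and the Fort_Events sound class at slightly reduced volume, and partially sends to a weapon-volume sidechain audio bus.
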SoundCue /PistolCanary_Event_Audio/Assets/Section02/tacaud1000/SC_PistolCanary_Section02_tacaud1000_Attack03
root=WavePlayer_0
nodes [{id=WavePlayer_0, properties {soundWaveAssetPtr=/PistolCanary_Event_Audio/Assets/Section02/wav/PistolCanary_Tentacle_Attk_1000_03}}]
properties {
 attenuationSettings=/PistolCanary_Audio/Assets/Settings/ATT_PistolCanary_Tentacle_Tip
 maxDistance=100000
 soundClassObject=&Fort_Events
 volumeMultiplier=0.8
 busSends [{audioBus=&AB_Sidechain_FortWeapon_Vol, minSendLevel=0, sendLevel=0.2, soundSourceBus=null}]
}
