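# Sound cue for the ground impact of a tentacle attack in Section 02 of the PistolCanary event.
# A Random node picks with equal weight among four wave players: three numbered attack variations and a FinalShot wave.
# The cue plays quieter and pitched down, routed to the Fort_Events sound class.
# It uses the shared PistolCanary projectile attenuation settings with a long audible range.
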SoundCue /PistolCanary_Event_Audio/Assets/Section02/tacaud1000/SC_PistolCanary_Section02_tacaud1000_GroundImpact
root=Random_0
nodes [
 {id=Random_0, children [WavePlayer_0,WavePlayer_1,WavePlayer_2,WavePlayer_3], properties {weights [1,1,1,1]}}
 {id=WavePlayer_0, properties {soundWaveAssetPtr=/PistolCanary_Event_Audio/Assets/Section02/wav/PistolCanary_Tentacle_Attk_1000_FinalShot}}
 {id=WavePlayer_1, properties {soundWaveAssetPtr=/PistolCanary_Event_Audio/Assets/Section02/wav/PistolCanary_Tentacle_Attk_1000_01}}
 {id=WavePlayer_2, properties {soundWaveAssetPtr=/PistolCanary_Event_Audio/Assets/Section02/wav/PistolCanary_Tentacle_Attk_1000_02}}
 {id=WavePlayer_3, properties {soundWaveAssetPtr=/PistolCanary_Event_Audio/Assets/Section02/wav/PistolCanary_Tentacle_Attk_1000_03}}
]
properties {
 attenuationSettings=/PistolCanary_Audio/Assets/Settings/ATT_PistolCanary_Projectiles
 maxDistance=57000
 pitchMultiplier=0.85
 soundClassObject=&Fort_Events
 volumeMultiplier=0.3
}
